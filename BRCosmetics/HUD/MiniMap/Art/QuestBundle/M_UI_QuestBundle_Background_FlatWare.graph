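# Translucent UI material for the FlatWare quest bundle background in the minimap HUD, with Focus and Select color states.
# Named reroutes carry an eased linear gradient, a level-adjusted background texture and an offset overlay texture.
# A color branch applies a Burst-driven overlay blend and lerps between state tints.
# Separate branches build a masked, screen-aligned panning triangle grid and a box whose corners round on one side with Select.
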Material /BRCosmetics/HUD/MiniMap/Art/QuestBundle/M_UI_QuestBundle_Background_FlatWare
graph {
 properties {blendMode=BLEND_Translucent, materialDomain=MD_UI, pixelDepthOffsetMode=PDOM_Legacy, usedShadingModels=Default Lit}
 nodes [
  LinearInterpolate(
   A=VectorParameter() {group=colors, parameterName=Base_RGB1, defaultValue=<rgba 0.0003 0.0185 0.2747 1>}.RGB
   B=VectorParameter() {group=colors, parameterName=Base_RGB2, defaultValue=<rgba 0 0.2242 1 1>}.RGB
  )
  n1 = NamedRerouteUsage()
  n2 = LinearInterpolate(A=Multiply(A=LinearInterpolate(A=@n1) {constAlpha=0.0001}), B=Add(A=NamedRerouteUsage()) {constB=0.5}) {constAlpha=0.4}
  n3 = NamedRerouteUsage()
  n4 = MaterialFunctionCall(Base=@n3, Blend=Constant() {r=1}) {materialFunction=&Blend_Overlay}
  n5 = Subtract(
   A=MaterialFunctionCall() {materialFunction=&GetUserInterfaceUV}.Normalized UV
   B=LinearInterpolate(
    A=ScalarParameter() {group=Triangles, parameterName=Busrt_Position_Start}
    B=ScalarParameter() {group=Triangles, parameterName=Busrt_Position_End}
    Alpha=ScalarParameter() {parameterName=Burst, sliderMax=1}
   )
  )
  n6 = Saturate(None=Multiply(A=ComponentMask(None=@n5) {r=true}, B=ScalarParameter() {defaultValue=0.3, group=Triangles, parameterName=Burst_sharpness}))
  n7 = LinearInterpolate(A=NamedRerouteUsage()) {constAlpha=0.0001}
  n8 = LinearInterpolate(A=LinearInterpolate(A=Multiply(A=@n7)) {constAlpha=0.4}) {constAlpha=0.01}
  n9 = LinearInterpolate(
   A=VectorParameter() {group=colors, parameterName=OffFocus_RGB1, defaultValue=<rgba 1 1 1 1>}.RGB
   B=VectorParameter() {group=colors, parameterName=OffFocus_RGB2, defaultValue=<rgba 1 1 1 1>}.RGB
  )
  n10 = LinearInterpolate(A=NamedRerouteUsage()) {constAlpha=0.0001}
  n11 = LinearInterpolate(A=LinearInterpolate(A=Multiply(A=@n10)) {constAlpha=0.4}) {constAlpha=0.01}
  n12 = LinearInterpolate(
   A=VectorParameter() {group=colors, parameterName=Selected_RGB1, defaultValue=<rgba 1 1 1 1>}.RGB
   B=VectorParameter() {group=colors, parameterName=Selected_RGB2, defaultValue=<rgba 1 1 1 1>}.RGB
  )
  LinearInterpolate(
   A=Multiply(A=LinearInterpolate(A=@n2) {constAlpha=0.01}, B=LinearInterpolate(A=@n3, B=@n4.Result, Alpha=@n6))
   B=LinearInterpolate(A=Multiply(A=@n8, B=@n9), B=Multiply(A=@n11, B=@n12), Alpha=ScalarParameter() {defaultValue=1, parameterName=Focus, sliderMax=1})
   Alpha=ScalarParameter() {parameterName=Select, sliderMax=1}
  )
  n13 = Constant()
  n14 = ComponentMask(None=TextureCoordinate()) {r=true}
  n15 = LinearInterpolate(
   B=ScalarParameter() {defaultValue=0.2, parameterName=Corner Roundness}
   Alpha=Multiply(A=Round(None=@n14), B=ScalarParameter() {parameterName=Select, sliderMax=1})
  )
  MaterialFunctionCall(
   Size=Constant() {r=1}
   CornerRoundness=MaterialFunctionCall(R=@n13, G=@n13, B=@n15, A=@n15) {materialFunction=&AppendMany}.RGBA
   KeepAspectRatio=StaticBool() {value=true}
  ) {materialFunction=&MF_UI_Box}
  n16 = MaterialFunctionCall(Float2=MaterialFunctionCall() {materialFunction=&ScreenResolution}.Visible Resolution) {materialFunction=&BreakOutFloat2Components}
  n17 = Divide(A=@n16.R, B=Max(A=@n16.G))
  n18 = Constant() {r=1}
  n19 = Rotator(
   Coordinate=Multiply(
    A=AppendVector(A=@n17, B=@n18)
    B=MaterialFunctionCall() {materialFunction=&ScreenAlignedUVs}."X 100%, Y 100%"
   )
   Time=ScalarParameter() {defaultValue=-0.0571, group=Triangles, parameterName=Tri Angle}
  ) {speed=6.2832}
  n20 = AppendVector(A=ScalarParameter() {defaultValue=0.0025, group=Triangles, parameterName=PanSpeed}, B=Constant())
  n21 = TextureSample(UVs=Multiply(A=Panner(Coordinate=@n19, Speed=@n20), B=ScalarParameter() {defaultValue=5, group=Triangles, parameterName=Tri Tile Amount})) {texture=/Game/Global/Textures/Default/Masks_and_Gradients/T-TriangleGrid}
  n22 = LinearInterpolate(
   A=ScalarParameter() {defaultValue=0.2, group=Triangles, parameterName=TriangleDensity}
   B=ScalarParameter() {defaultValue=0.2, group=Triangles, parameterName=TriangleDensity_Select}
   Alpha=ScalarParameter() {parameterName=Select, sliderMax=1}
  )
  n23 = ComponentMask(None=Subtract(
    A=MaterialFunctionCall() {materialFunction=&GetUserInterfaceUV}.Normalized UV
    B=ScalarParameter() {group=Triangles, parameterName=TriangleMask_Position}
   )) {r=true}
  n24 = ScalarParameter() {defaultValue=1.5, group=Triangles, parameterName=TriangleMask_sharpness}
  Multiply(A=Saturate(None=Multiply(A=@n21.R, B=@n22)), B=Saturate(None=Multiply(A=@n23, B=@n24)))
  n25 = VectorParameter() {parameterName=LinearGradientBottom, defaultValue=<rgba 0.7231 0.0976 0 1>}
  n26 = VectorParameter() {parameterName=LinearGradientA, defaultValue=<rgba 1 0.5711 0.4621 1>}
  n27 = VectorParameter() {parameterName=LinearGradientColourB, defaultValue=<rgba 1 0.5691 0 1>}
  n28 = MaterialFunctionCall("Value (0-1)"=ComponentMask(None=MaterialFunctionCall() {materialFunction=&GetUserInterfaceUV}.Normalized UV) {g=true}) {materialFunction=&MF_UI_EaseCurves}
  n29 = ScalarParameter() {defaultValue=0.06}
  n30 = ScalarParameter() {defaultValue=0.4, parameterName=Param_1}
  n31 = ScalarParameter() {defaultValue=0.25, parameterName=Param_2}
  n32 = ScalarParameter() {defaultValue=0.01, parameterName=Param_3}
  NamedRerouteDeclaration(None=LinearInterpolate(
    A=Multiply(A=@n25.RGB) {constB=0.7}
    B=LinearInterpolate(A=@n26.RGB, B=@n27.RGB, Alpha=@n28.EaseOutCubic)
    Alpha=MaterialFunctionCall(GradientStart=@n29, GradientEnd=@n30, RotationAngle=@n31, GradientOffset=@n32) {materialFunction=&MF_UI_LinearGradient}.Result
   )) {name=LinearGradient, nodeColor=<rgba 0.8047 1 0 1>}
  n33 = ScalarParameter() {defaultValue=0.13, parameterName=OverlayTextureX}
  n34 = ScalarParameter() {defaultValue=-0.3, parameterName=OverlayTextureY}
  n35 = MaterialFunctionCall(
   UVs=Add(A=MaterialFunctionCall() {materialFunction=&GetUserInterfaceUV}.Normalized UV, B=AppendVector(A=@n33, B=@n34))
   Texture Scale=AppendVector(
    A=ScalarParameter() {defaultValue=0.68, parameterName=OverlayTextureScaleX}
    B=ScalarParameter() {defaultValue=0.68, parameterName=OverlayTextureScaleY}
   )
  ) {materialFunction=&ScaleUVsByCenter}
  NamedRerouteDeclaration(None=TextureSampleParameter2D(UVs=MaterialFunctionCall(UVs=@n35.UVs) {materialFunction=&ScaleUVsByCenter}.UVs) {parameterName=AdditionalOverlayTexture, texture=/FortniteCoreUI/HUD/MiniMap/Art/QuestBundle/T_UI_FlatWare_TrackSmall}.RGB) {name=AdditionalOverlayTexture, nodeColor=<rgba 0 1 0.5313 1>}
  n36 = MaterialFunctionCall(UV=MaterialFunctionCall(
    UVs=MaterialFunctionCall() {materialFunction=&GetUserInterfaceUV}.9-Slice UV
    Texture Scale=ScalarParameter() {defaultValue=1, parameterName=BGScale}
   ) {materialFunction=&ScaleUVsByCenter}.UVs) {materialFunction=&MF_UI_UVScaling}
  n37 = TextureObjectParameter() {parameterName=Background Texture, texture=/FortniteCoreUI/HUD/MiniMap/Art/QuestBundle/T_UI_FlatWare_BG}
  n38 = Multiply(
   A=Power(
    Base=TextureSample(UVs=@n36.Scaled To Fill, Tex=@n37) {texture=&T_UI_FlatWare_TrackSmall}.RGB
    Exp=ScalarParameter() {defaultValue=0.6, parameterName=AdjustLevels}
   )
   B=ScalarParameter() {defaultValue=2, parameterName=TextureOpacity}
  )
  NamedRerouteDeclaration(None=Max(A=Saturate(None=@n38)) {constB=0.0001}) {name=BGTexture, nodeColor=<rgba 1 0 0.375 1>}
 ]
}
