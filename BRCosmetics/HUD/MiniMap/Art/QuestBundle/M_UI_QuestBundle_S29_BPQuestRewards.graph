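# Translucent UI material for Season 29 Battle Pass quest reward bundle art, drawing a portrait over a stylized background or only a border.
# The aspect-corrected portrait zooms on hover and fades toward its bottom.
# The background color-dodges a screen-aligned, panning triangle grid and adds a radial gradient.
# Box SDF rounded corners and rotated-UV angled masks flip with left alignment, while a static switch selects a selection-tinted, hover-revealed border variant.
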Material /BRCosmetics/HUD/MiniMap/Art/QuestBundle/M_UI_QuestBundle_S29_BPQuestRewards
graph {
 properties {blendMode=BLEND_Translucent, materialDomain=MD_UI, pixelDepthOffsetMode=PDOM_Legacy, usedShadingModels=Default Lit}
 nodes [
  n1 = AppendVector(
   A=ScalarParameter() {defaultValue=1, group=Dont edit on instances, parameterName=Background Scale X, sortPriority=100}
   B=ScalarParameter() {defaultValue=1, group=Dont edit on instances, parameterName=Background Scale Y}
  )
  NamedRerouteDeclaration(None=MaterialFunctionCall!(
    Texture Scale=NamedRerouteDeclaration(None=@n1) {name=outer box scale, nodeColor=<rgba 1 0 0.8203 1>}
    UVs=NamedRerouteUsage()
    Offset XY Texturesize Z=Constant3Vector() {constant=<rgba 1 1 1 0>}
   ) {materialFunction=/Engine/Functions/Engine_MaterialFunctions01/Texturing/ScaleUVsAroundPoint}.UVs) {name=background UVs, nodeColor=<rgba 0 0.5781 1 1>}
  n2 = MaterialFunctionCall(Float2=MaterialFunctionCall() {materialFunction=&GetUserInterfaceUV}.Pixel Size) {materialFunction=&BreakOutFloat2Components}
  n3 = Divide(A=@n2.R, B=Max(A=@n2.G))
  n4 = MaterialFunctionCall(Float2=NamedRerouteUsage()) {materialFunction=&BreakOutFloat2Components}
  n5 = Multiply(A=@n3, B=@n4.R)
  n6 = Divide(A=OneMinus(None=@n3))
  n7 = ScalarParameter() {group=Item, parameterName=Itemposition_U}
  n8 = Multiply(
   A=ScalarParameter() {defaultValue=0.1, group=Item, parameterName=Left aligned horizontal compensation}
   B=ScalarParameter() {group=Dont edit on instances, parameterName=Left aligned, sliderMax=1}
  )
  n9 = Add(
   A=AppendVector(A=Add(A=@n5, B=@n6), B=@n4.G)
   B=AppendVector(A=Add(A=@n7, B=@n8), B=ScalarParameter() {group=Item, parameterName=Itemposition_V})
  )
  n10 = ScalarParameter() {group=Item, parameterName=item rotation}
  n11 = ScalarParameter() {defaultValue=0.85, group=control, parameterName=Zoom}
  n12 = Add(A=@n11, B=ScalarParameter() {defaultValue=0.15, group=control, parameterName=hoverZoom})
  n13 = ScalarParameter() {group=control, parameterName=hover, sliderMax=1}
  n14 = TextureSampleParameter2D(UVs=MaterialFunctionCall(UVs=Rotator(Coordinate=@n9, Time=@n10), Texture Scale=LinearInterpolate(A=@n11, B=@n12, Alpha=@n13) {constA=1, constB=1.01}) {materialFunction=&ScaleUVsByCenter}.UVs) {group=Item, parameterName=NPC-Portrait, samplerSource=SSM_Clamp_WorldGroupSettings, texture=/Game/Athena/HUD/Quests/Art/T_NPC_Default}
  NamedRerouteDeclaration(None=@n14.RGB) {name=item color, nodeColor=<rgba 1 0 0.8906 1>}
  n15 = LinearInterpolate(
   A=ScalarParameter() {defaultValue=0.7, group=Dont edit on instances, parameterName=Right-Angle}
   B=ScalarParameter() {defaultValue=0.7, group=Dont edit on instances, parameterName=RIght-Angle-hover}
   Alpha=ScalarParameter() {group=control, parameterName=hover, sliderMax=1}
  )
  n16 = LinearInterpolate(Alpha=ScalarParameter() {group=Dont edit on instances, parameterName=Left aligned, sliderMax=1}) {constA=1, constB=-1, desc=flip angle if left aligned}
  n17 = ComponentMask(None=Rotator(Coordinate=NamedRerouteUsage(), Time=Multiply(A=@n15, B=@n16)) {centerX=0, centerY=0, speed=0.2}) {r=true}
  n18 = Multiply(
   A=LinearInterpolate(A=@n17, B=OneMinus(None=@n17), Alpha=ScalarParameter() {group=Dont edit on instances, parameterName=Left aligned, sliderMax=1}) {constB=0}
   B=ScalarParameter() {defaultValue=100, group=Dont edit on instances, parameterName=strenght}
  )
  NamedRerouteDeclaration(None=Saturate(None=Add(A=@n18) {constB=-10})) {name=inner bar, nodeColor=<rgba 1 0 0.3281 1>}
  NamedRerouteDeclaration(None=Saturate(None=@n18)) {name=outer bar, nodeColor=<rgba 1 0 0.4453 1>}
  n19 = MaterialFunctionCall(Float2=Multiply(A=MaterialFunctionCall() {materialFunction=&GetUserInterfaceUV}.Pixel Size, B=NamedRerouteUsage())) {materialFunction=&BreakOutFloat2Components}
  NamedRerouteDeclaration(None=AppendVector(A=Divide(A=@n19.R, B=@n19.R), B=Constant() {r=1})) {name=background aspect ratio, nodeColor=<rgba 0.5469 1 0 1>}
  NamedRerouteDeclaration(None=TextureCoordinate() {coordinateIndex=4}) {name=box UVs, nodeColor=<rgba 0 1 0.1328 1>}
  n20 = Constant()
  n21 = Add(A=@n20, B=Constant() {r=0.02}) {constB=0.02}
  n22 = ScalarParameter() {group=Dont edit on instances, parameterName=Corner roundness}
  n23 = ScalarParameter() {group=Dont edit on instances, parameterName=Left aligned, sliderMax=1}
  n24 = LinearInterpolate(B=@n22, Alpha=@n23) {constB=0}
  n25 = LinearInterpolate(A=@n22, Alpha=@n23) {constB=0}
  MaterialExpressionAppend4Vector_0 = MaterialXAppend4Vector!(A=@n24, B=@n24, C=@n25, D=@n25)
  n26 = Ceil(None=Max(A=@n24, B=@n25))
  n27 = MaterialFunctionCall(
   UVs=NamedRerouteUsage()
   Size=LinearInterpolate(Alpha=Saturate(None=@n26)) {constA=1, constB=0.99}
   CornerRoundness=@MaterialExpressionAppend4Vector_0
   KeepAspectRatio=StaticBool() {value=true}
  ) {materialFunction=&MF_UI_Box}
  NamedRerouteDeclaration(None=OneMinus(None=SmoothStep(Min=@n20, Max=@n21, Value=@n27.SDF) {constMax=0.02})) {name=NewRoundedCorners, nodeColor=<rgba 0.875 1 0 1>}
  n28 = Constant()
  n29 = Constant() {r=0.02}
  n30 = NamedRerouteUsage()
  n31 = Constant() {r=1}
  n32 = ComponentMask(None=NamedRerouteUsage()) {r=true}
  n33 = LinearInterpolate(
   B=ScalarParameter() {group=Dont edit on instances, parameterName=Corner roundness}
   Alpha=LinearInterpolate(A=@n32, B=OneMinus(None=@n32), Alpha=ScalarParameter() {group=Dont edit on instances, parameterName=Left aligned, sliderMax=1}) {constB=0}
  ) {constB=0}
  n34 = StaticBool() {value=true}
  NamedRerouteDeclaration(None=SmoothStep(
    Min=@n28
    Max=Add(A=@n28, B=@n29) {constB=0.02}
    Value=MaterialFunctionCall(UVs=@n30, Size=@n31, CornerRoundness=@n33, KeepAspectRatio=@n34) {materialFunction=&MF_UI_Box}.Fill
   ) {constMax=0.02}) {name=OldRoundedCorners, nodeColor=<rgba 0.875 1 0 1>}
  NamedRerouteDeclaration(None=Min(A=NamedRerouteUsage(), B=NamedRerouteUsage())) {name=outer box alpha, nodeColor=<rgba 0 1 0.4141 1>}
  n35 = VectorParameter() {group=Colors, parameterName=RGB1, defaultValue=<rgba 0 0.049 0.3646 1>}
  n36 = MaterialFunctionCall(Float2=MaterialFunctionCall() {materialFunction=&ScreenResolution}.Visible Resolution) {materialFunction=&BreakOutFloat2Components}
  n37 = Max(A=@n36.G)
  n38 = Multiply(
   A=AppendVector(A=Divide(A=@n36.R, B=@n37), B=Constant() {r=1})
   B=MaterialFunctionCall() {materialFunction=&ScreenAlignedUVs}."X 100%, Y 100%"
  )
  n39 = ScalarParameter() {defaultValue=-0.0571, group=Triangles, parameterName=Tri Angle}
  n40 = ScalarParameter() {defaultValue=0.0025, group=Triangles, parameterName=PanSpeed}
  n41 = Constant()
  n42 = Multiply(
   A=Panner(Coordinate=Rotator(Coordinate=@n38, Time=@n39) {speed=6.2832}, Speed=AppendVector(A=@n40, B=@n41))
   B=ScalarParameter() {defaultValue=5, group=Triangles, parameterName=Tri Tile Amount}
  )
  n43 = LinearInterpolate(
   A=@n35.RGB
   B=VectorParameter() {group=Colors, parameterName=RGB2, defaultValue=<rgba 0 0.4337 1 1>}.RGB
   Alpha=Multiply(
    A=TextureSample(UVs=@n42) {texture=/Game/Global/Textures/Default/Masks_and_Gradients/T-TriangleGrid}.R
    B=ScalarParameter() {defaultValue=0.2, group=Triangles, parameterName=TriangleDensity}
   )
  )
  n44 = MaterialFunctionCall(
   UVs=NamedRerouteUsage()
   CenterPosition=AppendVector(
    A=ScalarParameter() {defaultValue=0.5, group=Radial gradient, parameterName=RadialGradient_u, sliderMin=0.5}
    B=ScalarParameter() {defaultValue=0.9, group=Radial gradient, parameterName=RadialGradient_V, sliderMin=0.5}
   )
   Radius=ScalarParameter() {defaultValue=0.85, group=Radial gradient, parameterName=Radial Gradient Radius_V}
   Density=ScalarParameter() {defaultValue=1, group=Radial gradient, parameterName=Radial Gradient Density}
  ) {materialFunction=&RadialGradientExponential}
  NamedRerouteDeclaration(None=LinearInterpolate(
    A=MaterialFunctionCall!(Blend=@n43, Base=@n35.RGB) {materialFunction=/Engine/Functions/Engine_MaterialFunctions03/Blends/Blend_ColorDodge}.Result
    B=VectorParameter() {group=Colors, parameterName=RGB2, defaultValue=<rgba 0 0.4337 1 1>}.RGB
    Alpha=Saturate(None=@n44.RadialGradientExponential)
   )) {name=item BG color, nodeColor=<rgba 0.5234 0 1 1>}
  n45 = NamedRerouteDeclaration(None=@n14.A) {name=item alpha, nodeColor=<rgba 0.9922 1 0 1>}
  n46 = ComponentMask(None=NamedRerouteUsage()) {g=true}
  n47 = LinearInterpolate(
   A=ScalarParameter() {defaultValue=0.6, group=Fade, parameterName=item_bottom_position}
   Alpha=ScalarParameter() {group=control, parameterName=hover, sliderMax=1}
  ) {constA=1}
  n48 = OneMinus(None=Multiply(A=Subtract(A=@n46, B=@n47), B=ScalarParameter() {defaultValue=1, group=Fade, parameterName=item_bottom_strenght}))
  NamedRerouteDeclaration(None=Multiply(A=Saturate(None=@n45), B=Saturate(None=@n48))) {name=item alpha faded, nodeColor=<rgba 1 0 0.5391 1>}
  NamedRerouteUsage()
  StaticBool()
  StaticSwitchParameter(
   True=LinearInterpolate(
    A=VectorParameter() {group=Background, parameterName=Border Color Unselected, defaultValue=<rgba 0 0.1714 1 1>}.RGB
    B=VectorParameter() {group=Background, parameterName=Border Color Selected, defaultValue=<rgba 1 1 1 1>}.RGB
    Alpha=ScalarParameter() {group=Dont edit on instances, parameterName=Border IsSelected, sliderMax=1}
   )
   False=LinearInterpolate(A=NamedRerouteUsage(), B=NamedRerouteUsage(), Alpha=NamedRerouteUsage())
  ) {parameterName=Only border}
  n49 = NamedRerouteUsage()
  StaticSwitchParameter(True=LinearInterpolate(B=@n49, Alpha=ScalarParameter() {group=control, parameterName=hover, sliderMax=1}), False=@n49) {parameterName=Only border}
 ]
}
